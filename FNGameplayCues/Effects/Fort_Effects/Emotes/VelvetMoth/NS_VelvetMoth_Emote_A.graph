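# Niagara system for the VelvetMoth emote, wrapping the character in warm orange glowing rays and rising sparkles.
# A bone-attached ray mesh, curl-noise sparkles and two sprite emitters distributed along bone chains via FN_BoneFlow with pulsing radial streaks form the effect.
# System flare and brightness values drive ray and streak color, while time of day scales sparkle and streak alpha.
# Digit counter, beam, halo and center sparkle emitters are disabled.
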NiagaraSystem /FNGameplayCues/Effects/Fort_Effects/Emotes/VelvetMoth/NS_VelvetMoth_Emote_A
dependencies {
 usedDataInterfaces [
  &NiagaraDataInterfaceColorCurve
  &NiagaraDataInterfaceCurve
  &NiagaraDataInterfaceVector2DCurve
  &NiagaraDataInterfaceArrayInt32
  &NiagaraDataInterfaceArrayFloat3
  &NiagaraDataInterfaceSkeletalMesh
 ]
 usedDynamicInputs [&ModuloFloat, &ClampFloat]
 usedModules [
  &self:SystemSpawnScript.SystemScriptSource.SystemScriptGraph.NiagaraNodeAssignment_5.SetVariables_#1
  &self:SystemSpawnScript.SystemScriptSource.SystemScriptGraph.NiagaraNodeAssignment_6.SetVariables_#2
  &SystemState
  &Timeline
  &self:SystemSpawnScript.SystemScriptSource.SystemScriptGraph.NiagaraNodeAssignment_4.SetVariables_#3
  &self:SampleSingleBoneSpeed
  &self:SystemSpawnScript.SystemScriptSource.SystemScriptGraph.NiagaraNodeAssignment_0.SetVariables_#4
  &self:SystemSpawnScript.SystemScriptSource.SystemScriptGraph.NiagaraNodeAssignment_1.SetVariables_#5
  &self:SystemSpawnScript.SystemScriptSource.SystemScriptGraph.NiagaraNodeAssignment_3.SetVariables_#6
  &self:SystemSpawnScript.SystemScriptSource.SystemScriptGraph.NiagaraNodeAssignment_2.SetVariables_#7
 ]
}
emitters {
 Body_Rays_Mesh {
  emitterUpdateScript [
   &EmitterState(
    Inactive Response=::NewEnumerator2
    Life Cycle Mode=::NewEnumerator1
    Loop Behavior=::NewEnumerator1
    Loop Duration=1
    Scalability Mode=::NewEnumerator1
    UseLoopDelay=false
   )
   &SpawnBurst_Instantaneous(Use Spawn Probability=false, UseLoopCountLimit=false)
  ]
  particleSpawnScript [&InitializeParticle(
    Color=<rgba 1 0.4 0 0.3>
    Lifetime=125
    Sprite Size Mode=::NewEnumerator3
    Uniform Sprite Size=75
    UsePositionOffset=false
    Write Material Random=true
   )]
  particleUpdateScript [
   &SkeletalMeshLocation(
    Bone / Socket Index=5
    Bone Sampling Mode=::NewEnumerator4
    MaskSpawnGroup=false
    Orientation Sampling=::NewEnumerator1
    Sampled Position Offset=<0 0 15>
    Sampled Position Offset Coordinate Space:ENiagaraSkelMeshTransforms=::NewEnumerator1
    Skeletal Mesh=$System.Skel
   ) {module=SkeletalMeshLocation001}
   &CameraOffset(Camera Offset Amount=-10)
   /SetVariables(Particles.Scale=<4 4 8>) {isSetParameters=true}
   /SetVariables(Particles.SpriteSize=<50 60>) {isSetParameters=true}
   &DynamicMaterialParameters(
    Index 0 Param 1=&Multiply_Float(A=&MakeFloatFromInt(Int=&ReturnExecIndex()), B=22)
    Index 0 Param 2=1
    Index 0 Param 3=5
    Index 0 Param 4=5
    Module.Param0WriteEnabled=true
    Module.Param1WriteEnabled=false
    Module.Param2WriteEnabled=true
    Module.Param3WriteEnabled=false
   )
   /SetVariables(Particles.ColorB:LinearColor=<rgba 1 0.5 0 1>) {isSetParameters=true}
   &ScaleColor(
    Scale RGB=&VectorFromFloat(Value=&Add_Float(A=1, B=$System.FlareAll))
    ScaleA=false
    ScaleRGB=true
   )
   &ScaleColor(
    Scale Alpha=$System.FXBrightnessRamped
    Scale RGB=&VectorFromFloat(Value=$System.FXBrightnessRamped)
    ScaleA=true
    ScaleRGB=true
   ) {module=ScaleColor003}
  ]
  rendererClasses [&NiagaraMeshRendererProperties]
  renderers [{rendererClass=&NiagaraMeshRendererProperties, rendererIndex=0}]
 }
 Body_Sparkles {
  emitterUpdateScript [
   &EmitterState()
   &SpawnRate(
    SpawnRate=&MakeCustomFloatFromBool(Bool=&SetBoolByFloatComparison(A=1, B=0.2), True Float=50, False Float=0)
    Use Spawn Probability=false
   )
  ]
  particleSpawnScript [
   &InitializeParticle(
    Color Maximum=<rgba 0.9039 1 0.6539 1>
    Color Minimum=<rgba 1 0.3801 0 1>
    Color Mode=::NewEnumerator2
    Lifetime=&RandomRangeFloat(Minimum=0.5, Maximum=1)
    Sprite Rotation Mode=::NewEnumerator2
    Sprite Size Mode=::NewEnumerator1
    Uniform Sprite Size Max=5
    Uniform Sprite Size Min=3
    UsePositionOffset=false
    Write Material Random=true
   )
   /SetVariables(Particles.RandomCore:NiagaraFloat=&FloatFromCurve(
     FloatCurve=DataInterface() {propertyValues {Curve {keys [
         {arriveTangent=0.0559, interpMode=RCIM_Cubic, leaveTangent=0.0559, tangentMode=RCTM_User, time=0, value=0.15}
         {arriveTangent=0.1895, interpMode=RCIM_Cubic, leaveTangent=0.1895, tangentMode=RCTM_User, time=0.9, value=0.3}
         {arriveTangent=1.1808, interpMode=RCIM_Cubic, leaveTangent=1.1808, tangentMode=RCTM_User, time=1, value=0.4}
        ]}}}
     CurveIndex=&RandomRangeFloat(Minimum=0, Maximum=1)
     Scale Curve=2
    )) {isSetParameters=true, module=SetVariables_#8}
   &SkeletalMeshLocation(Bone / Socket Index=2, Bone Sampling Mode=::NewEnumerator4, MaskSpawnGroup=false, Skeletal Mesh=$System.Skel)
   &SphereLocation(
    MaskSpawnGroup=false
    Override Seed=false
    Sphere Radius=35
    Surface Only=true
    UseNonUniformScale=false
    UseOffset=false
    UseSphereOrientation=false
   ) {module=SphereLocation001}
   &AddVelocityFromPoint(Module.Use Falloff Distance=false, Origin Offset=<0 0 -50>, Velocity Strength=5)
   &AddVelocity(Velocity=&Multiply_VectorByFloat(Vector=$Output.SphereLocation001.SphereNormal, Float=15))
   &AddVelocity(Velocity=<0 -100 0>) {enabled=false, module=AddVelocity002}
   &ApplyInitialForces()
  ]
  particleUpdateScript [
   &ParticleState()
   &CurlNoiseForce(Noise Frequency=4, Noise Strength=75, Pan Noise=true, Pan Noise Field=<0.5 0.5 0>)
   &GravityForce(Gravity=<0 0 255>)
   &ScaleSpriteSize(Scale Factor=&Vector2DFromFloat(Value=&FloatFromCurve(
      FloatCurve=DataInterface() {propertyValues {Curve {keys [<at 0 1 RCIM_Cubic>, {arriveTangent=-1.6315, interpMode=RCIM_Cubic, leaveTangent=-1.6315, tangentMode=RCTM_User, time=1, value=0}]}}}
      Scale Curve=$Particles.RandomCore
     )))
   &ScaleSpriteSize(
    Scale Factor=&Multiply_Vector2D(A=$Particles.SpriteSize, B=&Vector2DFromFloat(Value=0.6))
    Scale Sprite Size Mode=::NewEnumerator2
   ) {module=ScaleSpriteSize002}
   &ScaleColor(
    Scale Alpha=&Multiply_Float(
     A=&FloatFromCurve(FloatCurve=DataInterface() {propertyValues {Curve {keys [<at 0.8 1>,<at 1 0>]}}})
     B=&FNTimeOfDayScalar(Night Scale=0.3, Morning Scale=0.1, Evening Scale=0.2)
    )
    Scale RGB=&VectorFromFloat(Value=&FloatFromCurve(FloatCurve=DataInterface() {propertyValues {Curve {keys [<at 0 1>,<at 0.1 3>,<at 0.2 1>]}}}, Scale Curve=0.2))
    ScaleA=true
    ScaleRGB=true
   )
   &Drag(Drag=2, Use Linear Drag=true, Use Rotational Drag=false)
   &SolveForcesAndVelocity(Clamp Velocity=false, Limit Acceleration=false)
   &CameraOffset(Camera Offset Amount=-55)
  ]
  rendererClasses [&NiagaraSpriteRendererProperties]
  renderers [{rendererClass=&NiagaraSpriteRendererProperties}]
 }
 Counter_burst {
  bEnabled=false
  emitterUpdateScript [
   &EmitterState(Life Cycle Mode=::NewEnumerator1, Loop Behavior=::NewEnumerator1, Loop Duration=2, UseLoopDelay=false)
   &SpawnBurst_Instantaneous(Spawn Count=25, Spawn Time=1.5, Use Spawn Probability=false)
  ]
  particleSpawnScript [
   &InitializeParticle(
    Color=<rgba 1 0.3 0 1>
    Lifetime Max=0.4
    Lifetime Min=0.05
    Lifetime Mode=::NewEnumerator1
    Sprite Size Mode=::NewEnumerator3
    UsePositionOffset=false
    Write Material Random=true
   )
   &SkeletalMeshLocation(
    Bone / Socket Index=7
    Bone Sampling Mode=::NewEnumerator4
    MaskSpawnGroup=false
    Sampled Orientation Coordinate Space=::World
    Sampled Position Coordinate Space=::World
    Sampled Position Offset=<5 0 45>
    Sampled Position Offset Coordinate Space:ENiagaraSkelMeshTransforms=::NewEnumerator1
    Skeletal Mesh=$System.Skel
   ) {module=SkeletalMeshLocation001}
   &SphereLocation(
    MaskSpawnGroup=false
    Offset=<4 0 45>
    Override Seed=false
    Sphere Coordinate Space=::World
    Sphere Origin:NiagaraPosition=$Particles.SkeletalMeshLocation001.SampledPosition
    Sphere Radius=0.1
    Surface Only=false
    UseNonUniformScale=false
    UseOffset=true
    UseSphereOrientation=false
   )
   &AddVelocityFromPoint(Module.Use Falloff Distance=false, Origin Offset:NiagaraPosition=$Particles.SkeletalMeshLocation001.SampledPosition, Velocity Strength=555) {enabled=false}
   &AddVelocity(
    Coordinate Space=::World
    Velocity=&Multiply_VectorByFloat(Vector=$Output.SphereLocation.SphereNormal, Float=&RandomRangeFloat(Minimum=222, Maximum=0.2))
   )
  ]
  particleUpdateScript [
   &ParticleState()
   &GravityForce(Gravity=<0 0 15>)
   &CurlNoiseForce(Noise Frequency=1, Noise Strength=0.5, Pan Noise=false, Pan Noise Field=<1 0 3>)
   &SolveForcesAndVelocity(Clamp Velocity=false, Limit Acceleration=false)
   &SolveRotationalForcesAndVelocity() {enabled=false}
   &ScaleColor(
    Linear Color Curve=DataInterface() {propertyValues {
      AlphaCurve {keys [
        {arriveTangent=3.6054, interpMode=RCIM_Cubic, leaveTangent=3.6054, tangentMode=RCTM_User, time=0, value=0}
        <at 0.2 1 RCIM_Cubic>
        {arriveTangent=-1.25, interpMode=RCIM_Cubic, leaveTangent=-1.25, time=1, value=0}
       ]}
      BlueCurve {keys [
        {arriveTangent=3.7716, interpMode=RCIM_Cubic, leaveTangent=3.7716, tangentMode=RCTM_User, time=0.0097, value=0}
        <at 0.2 1 RCIM_Cubic>
        {arriveTangent=-1.25, interpMode=RCIM_Cubic, leaveTangent=-1.25, time=1, value=0}
       ]}
      GreenCurve {keys [
        {arriveTangent=3.7716, interpMode=RCIM_Cubic, leaveTangent=3.7716, tangentMode=RCTM_User, time=0.0097, value=0}
        <at 0.2 1 RCIM_Cubic>
        {arriveTangent=-1.25, interpMode=RCIM_Cubic, leaveTangent=-1.25, time=1, value=0}
       ]}
      RedCurve {keys [
        {arriveTangent=3.7716, interpMode=RCIM_Cubic, leaveTangent=3.7716, tangentMode=RCTM_User, time=0.0097, value=0}
        <at 0.2 1 RCIM_Cubic>
        {arriveTangent=-1.25, interpMode=RCIM_Cubic, leaveTangent=-1.25, time=1, value=0}
       ]}
     }}
    Scale Mode=::NewEnumerator2
   )
  ]
  rendererClasses [&NiagaraSpriteRendererProperties]
  renderers [{rendererClass=&NiagaraSpriteRendererProperties}]
 }
 Counter_nosim_infDigits {
  bEnabled=false
  emitterUpdateScript [
   &EmitterState(
    Life Cycle Mode=::NewEnumerator1
    Loop Behavior=::NewEnumerator1
    Loop Duration Mode=::NewEnumerator1
    Scalability Mode=::NewEnumerator1
    UseLoopDelay=false
   )
   &SpawnBurst_Instantaneous(Spawn Count=1, Spawn Time=1.55, Use Spawn Probability=false)
  ]
  particleSpawnScript [
   &InitializeParticle(
    Color=<rgba 0.4844 0.3043 0.0681 1>
    Mesh Scale Mode=::NewEnumerator3
    Sprite Size Mode=::NewEnumerator3
    Uniform Sprite Size=&Multiply_Float(
     A=64
     B=&MakeCustomFloatFromBool(
      Bool=&SetBoolByFloatComparison(A=&MakeFloatFromInt(Int=$User.InputDigits), B=999)
      True Float=1
      False Float=0
     )
    )
    UsePositionOffset=false
    Write Material Random=true
   )
   /SetVariables(Particles.Flamecolor:LinearColor=<rgba 0.1979 0.074 0.0135 1>) {isSetParameters=true}
  ]
  particleUpdateScript [
   &SkeletalMeshLocation(
    Bone / Socket Index=7
    Bone Sampling Mode=::NewEnumerator4
    MaskSpawnGroup=false
    Orientation Sampling=::NewEnumerator1
    Sampled Orientation Coordinate Space=::World
    Sampled Position Coordinate Space=::World
    Sampled Position Offset=<5 -45 0>
    Skeletal Mesh=$System.Skel
   ) {module=SkeletalMeshLocation001}
   &ParticleState(Kill Particles When Lifetime Has Elapsed=false)
   /SetVariables(Particles.RollingDigits:NiagaraInt32=&MakeIntFromFloat(Float=&FloatFromCurve(
      FloatCurve=DataInterface() {propertyValues {Curve {keys [<at 0 0>,<at 2 1>]}}}
      CurveIndex=&Add_Float(A=&Multiply_Float(A=$Emitter.Age, B=1), B=-1.5)
      Scale Curve=&MakeFloatFromInt(Int=/Niagara/DynamicInputs/Math/Min_Int(A:NiagaraInt32=$User.InputDigits, B:NiagaraInt32=99999))
     ))) {isSetParameters=true}
   &self:CalcDigits(count:NiagaraInt32=$Particles.RollingDigits) {module=CalcDigits}
   &DynamicMaterialParameters(
    Index 0 Param 1=4.1852
    Index 0 Param 2=1
    Index 0 Param 3=1
    Index 1 Param 1=$Output.CalcDigits.digits
    Index 1 Param 2=&MakeFloatFromInt(Int=$Particles.RollingDigits)
    Module.Param0WriteEnabled=true
    Module.Param0WriteEnabled1=true
    Module.Param1WriteEnabled=true
    Module.Param1WriteEnabled1=true
    Module.Param2WriteEnabled=true
    Module.Param2WriteEnabled1=false
    Module.Param3WriteEnabled=false
    Module.Param3WriteEnabled1=false
    Write Parameter Index 1=true
   ) {module=DynamicMaterialParameters001}
   &ScaleMeshSize(Scale Factor=&Multiply_Vector(
     A=$Particles.DigitScale
     B=&VectorFromFloat(Value=&FloatFromCurve(
       FloatCurve=DataInterface() {propertyValues {Curve {keys [
           {arriveTangent=4.1785, interpMode=RCIM_Cubic, leaveTangent=4.1785, tangentMode=RCTM_User, time=0, value=0}
           {arriveTangent=3.5479, interpMode=RCIM_Cubic, leaveTangent=3.5479, tangentMode=RCTM_User, time=0.1222, value=0.6825}
           {arriveTangent=-0.0405, interpMode=RCIM_Cubic, leaveTangent=-0.0405, tangentMode=RCTM_User, time=0.5136, value=1.152}
           <at 0.7042 1>
          ]}}}
       Scale Curve=2
      ))
    )) {module=ScaleMeshSize001}
  ]
  rendererClasses [&NiagaraMeshRendererProperties, &NiagaraSpriteRendererProperties]
  renderers [
   {rendererClass=&NiagaraMeshRendererProperties, rendererIndex=0}
   {rendererClass=&NiagaraSpriteRendererProperties, rendererIndex=1}
  ]
 }
 beam_elements {
  bEnabled=false
  simTarget=GPUComputeSim
  emitterUpdateScript [
   &EmitterState(
    Life Cycle Mode=::NewEnumerator1
    Loop Behavior=::NewEnumerator1
    Loop Duration Mode=::NewEnumerator1
    Scalability Mode=::NewEnumerator1
    UseLoopDelay=false
   )
   &SpawnBurst_Instantaneous(Spawn Time=0.55, Use Spawn Probability=false, UseLoopCountLimit=false)
  ]
  particleSpawnScript [&InitializeParticle(
    Color=<rgba 4 2 0 0.2>
    Lifetime=1.3
    Mesh Scale=<0.75 0.75 0.75>
    Mesh Scale Mode=::NewEnumerator4
    Sprite Size Mode=::NewEnumerator3
    Uniform Sprite Size=75
    UsePositionOffset=false
    Write Material Random=true
   )]
  particleUpdateScript [
   &ParticleState(Kill Particles When Lifetime Has Elapsed=false)
   &SkeletalMeshLocation(
    Bone / Socket Index=7
    Bone Sampling Mode=::NewEnumerator4
    MaskSpawnGroup=false
    Orientation Sampling=::NewEnumerator1
    Sampled Position Offset=<8 -15 0>
    Skeletal Mesh=$System.Skel
   ) {module=SkeletalMeshLocation001}
   &CameraOffset(Camera Offset Amount=-15)
   /SetVariables(Particles.SpriteSize=&Multiply_Vector2D_ByFloat(
     Float=&FloatFromCurve(
      FloatCurve=DataInterface() {propertyValues {Curve {keys [
          {arriveTangent=-0.0094, interpMode=RCIM_Cubic, leaveTangent=-0.0094, tangentMode=RCTM_User, time=0, value=0}
          {arriveTangent=7.6747, interpMode=RCIM_Cubic, leaveTangent=7.6747, tangentMode=RCTM_User, time=0.3, value=2}
          {arriveTangent=-0.0405, interpMode=RCIM_Cubic, leaveTangent=-0.0405, tangentMode=RCTM_User, time=1, value=4.375}
         ]}}}
      Scale Curve=33
     )
     Vector2D=<1 1>
    )) {isSetParameters=true, module=SetVariables_#9}
   &Color(Color=&LerpLinearColors(
     StartColor=<rgba 5 4 0 0.2>
     EndColor=<rgba 3.0447 0 5 0.2>
     LerpFactor=&FloatFromCurve(FloatCurve=DataInterface() {propertyValues {Curve {keys [
          {arriveTangent=9.0725, interpMode=RCIM_Cubic, leaveTangent=9.0725, tangentMode=RCTM_User, time=0, value=0}
          {arriveTangent=-2.5249, interpMode=RCIM_Cubic, leaveTangent=-2.5249, tangentMode=RCTM_User, time=0.1408, value=0.9847}
          {arriveTangent=-0.0405, interpMode=RCIM_Cubic, leaveTangent=-0.0405, tangentMode=RCTM_User, time=1, value=0}
         ]}}})
    ))
   &ScaleColor(
    Scale Alpha=&FloatFromCurve(FloatCurve=DataInterface() {propertyValues {Curve {keys [
         {arriveTangent=-0.2933, interpMode=RCIM_Cubic, leaveTangent=-0.2933, tangentMode=RCTM_User, time=0, value=0}
         {arriveTangent=3.5997, interpMode=RCIM_Cubic, leaveTangent=3.5997, tangentMode=RCTM_User, time=0.1, value=0.95}
         {arriveTangent=-1.6622, interpMode=RCIM_Cubic, leaveTangent=-1.6622, tangentMode=RCTM_User, time=0.5, value=0.35}
         {arriveTangent=-0.0383, interpMode=RCIM_Cubic, leaveTangent=-0.0383, tangentMode=RCTM_User, time=0.9, value=0}
        ]}}})
    ScaleA=true
    ScaleRGB=false
   ) {module=ScaleColor001}
   &ScaleSpriteSize(Uniform Curve Sprite Scale=DataInterface() {propertyValues {Curve {
       postInfinityExtrap=RCCE_Cycle
       preInfinityExtrap=RCCE_Cycle
       keys [
        {interpMode=RCIM_Cubic, leaveTangentWeight=0.7084, tangentMode=RCTM_User, time=0, value=1}
        {
         arriveTangent=-0.0106
         arriveTangentWeight=0.3333
         interpMode=RCIM_Cubic
         leaveTangent=-0.0106
         leaveTangentWeight=0.3333
         time=2.1218
         value=0.879
        }
        {arriveTangent=0.01, interpMode=RCIM_Cubic, leaveTangent=0.01, time=5.2552, value=0.9441}
        {arriveTangent=0.0118, interpMode=RCIM_Cubic, leaveTangent=0.0118, time=7.1328, value=0.929}
        {arriveTangentWeight=0.956, interpMode=RCIM_Cubic, tangentMode=RCTM_User, time=10, value=1}
       ]
      }}})
   /SetVariables(Particles.SpriteSize_B:Vector2f=&Multiply_Vector2D_ByFloat(
     Float=&FloatFromCurve(
      FloatCurve=DataInterface() {propertyValues {Curve {
         postInfinityExtrap=RCCE_Cycle
         preInfinityExtrap=RCCE_Cycle
         keys [
          {interpMode=RCIM_Cubic, leaveTangentWeight=0.7374, tangentMode=RCTM_User, time=0, value=1}
          {
           arriveTangent=-0.0106
           arriveTangentWeight=0.5406
           interpMode=RCIM_Cubic
           leaveTangent=-0.0106
           tangentMode=RCTM_User
           time=2.1218
           value=0.879
          }
          {
           arriveTangent=0.01
           arriveTangentWeight=1.0445
           interpMode=RCIM_Cubic
           leaveTangent=0.01
           leaveTangentWeight=1.5816
           tangentMode=RCTM_User
           time=5.2552
           value=0.9441
          }
          {
           arriveTangent=0.0118
           arriveTangentWeight=0.6259
           interpMode=RCIM_Cubic
           leaveTangent=0.0118
           leaveTangentWeight=0.9557
           tangentMode=RCTM_User
           time=7.1328
           value=0.929
          }
          {arriveTangentWeight=2.6261, interpMode=RCIM_Cubic, tangentMode=RCTM_User, time=10, value=1}
         ]
        }}}
      Scale Curve=3
     )
     Vector2D=<15 123.75>
    )) {isSetParameters=true}
   &CameraOffset(Camera Offset Amount=15) {module=CameraOffset001}
   &self:SetAlignment(AimPosition:Vector3f=<0 0 1>) {module=SetAlignment}
  ]
  rendererClasses [&NiagaraSpriteRendererProperties, &NiagaraMeshRendererProperties]
  renderers [
   {rendererClass=&NiagaraSpriteRendererProperties}
   {rendererClass=&NiagaraMeshRendererProperties, rendererIndex=1}
  ]
 }
 boneflow_radial_peaks_B {
  emitterUpdateScript [
   &EmitterState(Life Cycle Mode=::NewEnumerator1, Loop Behavior=::NewEnumerator1, Loop Duration Mode=::NewEnumerator1, UseLoopDelay=false)
   &SpawnRate(
    SpawnRate=&MakeCustomFloatFromBool(
     Bool=&SetBoolByFloatComparison(A=&VectorLength(Vector=$Engine.Owner.Velocity), B=50)
     True Float=0
     False Float=50
    )
    Use Spawn Probability=false
   ) {enabled=false}
   &SpawnBurst_Instantaneous(Spawn Count=15, Use Spawn Probability=false, UseLoopCountLimit=false)
  ]
  particleSpawnScript [
   &InitializeParticle(
    Color=<rgba 1 0.6 0 0.05>
    Lifetime=&RandomRangeFloat(Minimum=1, Maximum=1)
    Sprite Size Mode=::NewEnumerator3
    Uniform Sprite Size=25
    UsePositionOffset=false
    Write Material Random=true
   )
   /SetVariables(Particles.T:NiagaraFloat=&RemapRange(
     Input Value To Remap=&ReturnNormalizedExecIndex(Normalized Index Includes One=false)
     Input Min=0
     Input Max=1
     Output Min=0
     Output Max=1
    )) {isSetParameters=true, module=SetVariables_#10}
  ]
  particleUpdateScript [
   &ParticleState(Kill Particles When Lifetime Has Elapsed=false)
   &self:Pulse_Rays_01(
    Frequency=3
    Time:NiagaraFloat=$Engine.Time
    Time Offset:NiagaraFloat=&ReturnNormalizedExecIndex(Normalized Index Scale=12)
   ) {module=Pulse_Rays_01}
   &ScaleSpriteSize(
    Non-Uniform Curve Index:NiagaraFloat=&ReturnNormalizedExecIndex()
    Non-Uniform Curve Scale=<3 -2>
    Non-Uniform Curve Sprite Scale=DataInterface() {propertyValues {
      XCurve {keys [
        {arriveTangent=2.5, interpMode=RCIM_Cubic, leaveTangent=2.5, time=0.2, value=-0.025}
        {arriveTangent=3.308, interpMode=RCIM_Cubic, leaveTangent=3.308, tangentMode=RCTM_User, time=0.3, value=0.225}
        {arriveTangent=0.3732, interpMode=RCIM_Cubic, leaveTangent=0.3732, time=0.4, value=0.575}
        {arriveTangent=-0.9433, interpMode=RCIM_Cubic, leaveTangent=-0.9433, time=0.7019, value=0.375}
        {arriveTangent=-1.219, interpMode=RCIM_Cubic, leaveTangent=-1.219, time=1.0095, value=0}
       ]}
      YCurve {keys [
        {arriveTangent=2.5, interpMode=RCIM_Cubic, leaveTangent=2.5, time=0.2, value=-0.025}
        {arriveTangent=3.308, interpMode=RCIM_Cubic, leaveTangent=3.308, tangentMode=RCTM_User, time=0.3, value=0.225}
        {arriveTangent=0.3732, interpMode=RCIM_Cubic, leaveTangent=0.3732, time=0.4, value=0.575}
        {arriveTangent=-0.9023, interpMode=RCIM_Cubic, leaveTangent=-0.9023, time=0.7019, value=0.375}
        {arriveTangent=-1.1377, interpMode=RCIM_Cubic, leaveTangent=-1.1377, time=1.0095, value=0.025}
       ]}
     }}
    Scale Sprite Size Mode=::NewEnumerator3
   ) {module=ScaleSpriteSize003}
   &FN_BoneFlow(
    Bias=0.3515
    Bone Offsets=DataInterface() {propertyValues {
      GpuSyncMode=SyncCpuToGpu
      MaxElements=0
      FloatData [<8 0 0>, <-6 0 0>, <1 0 0>, <6 0 0>, <0 0 0>, <0 0 0>, <0 0 0>, <0 0 0>, <0 0 0>, <0 0 0>, <0 0 0>, <0 0 0>]
     }}
    Bone Sets=DataInterface() {propertyValues {GpuSyncMode=SyncCpuToGpu, MaxElements=0, IntData [0,1,1,2,2,3,3,4]}}
    Enable Position Offsets=true
    Height:NiagaraFloat=$Particles.T
    SkeletalMesh:NiagaraDataInterfaceSkeletalMesh=$System.Skel
    Step Size:NiagaraFloat=&Divide_Float(A=0.1, B=1)
    Tension=0.3
    scale=1
   )
   &SampleSkeletalMesh(
    AdditionalYawPitchRoll=false
    Bone / Socket Index=2
    Bone Sampling Mode:ENiagaraBoneSamplingMode=::NewEnumerator4
    "Interpolate Mesh Sampling (Update)"=true
    Position Offset Coordinate Space:ENiagaraCoordinateSpace=::World
    Skeletal Mesh:NiagaraDataInterfaceSkeletalMesh=$System.Skel
   )
   &self:Orient_01(Offset:Vector3f=<0 0 35>, RadialCenter:NiagaraPosition=$StackContext.SampleSkeletalMesh.SampledPosition, offsetScale=0.2) {enabled=false, module=Orient_01}
   &self:ScratchModule_01(Offset:Vector3f=<0 0 -20>, RadialCenter:Vector3f=$Particles.SampleSkeletalMesh.SampledPosition) {enabled=false, module=ScratchModule_01}
   &self:ScratchModule_01(
    Offset:Vector3f=<0 0 -25>
    RadialCenter:Vector3f=$Particles.SampleSkeletalMesh.SampledPosition
    offsetScale:NiagaraFloat=&FloatFromCurve(FloatCurve=DataInterface() {propertyValues {Curve {keys [
         {arriveTangent=0.5893, interpMode=RCIM_Cubic, leaveTangent=0.5893, tangentMode=RCTM_User, time=0, value=-0.2}
         {interpMode=RCIM_Cubic, tangentMode=RCTM_User, time=0.5, value=-0.05}
        ]}}})
   ) {module=ScratchModule_01001}
   &CameraOffset()
   &ScaleSpriteSize(Scale Factor=&Vector2DFromFloat(Value=&FloatFromCurve(
      FloatCurve=DataInterface() {propertyValues {Curve {keys [<at 0 1 RCIM_Cubic>, {arriveTangent=-1.6315, interpMode=RCIM_Cubic, leaveTangent=-1.6315, tangentMode=RCTM_User, time=1, value=0}]}}}
      Scale Curve=$Particles.RandomCore
     ))) {enabled=false}
   &ScaleColor(
    Scale Alpha=&Multiply_Float(
     A=&FloatFromCurve(
      FloatCurve=DataInterface() {propertyValues {Curve {keys [<at 0.8 1>,<at 1 0>]}}}
      CurveIndex=&ReturnNormalizedExecIndex()
     )
     B=&FNTimeOfDayScalar(Night Scale=0.3, Morning Scale=0.1, Evening Scale=0.2)
    )
    Scale RGB=&VectorFromFloat(Value=&FloatFromCurve(
      FloatCurve=DataInterface() {propertyValues {Curve {keys [<at 0 0>,<at 1 1>]}}}
      CurveIndex=&ReturnNormalizedExecIndex()
      Scale Curve=&Lerp_Float(A=1, B=1, Alpha=$Particles.Pulse)
     ))
    ScaleA=true
    ScaleRGB=true
   )
   &self:OffsetZ_01(Offset:Vector3f=<0 0 33>, Pivot:Vector3f=$Particles.SampleSkeletalMesh.SampledPosition, Scale=2) {module=OffsetZ_01}
   &DynamicMaterialParameters(
    Index 0 Param 1=1
    Index 0 Param 2=5
    Index 0 Param 3=0.6
    Index 3 Param 1=1
    Index 3 Param 2=&MakeFloatFromInt(Int=&ReturnExecIndex())
    Module.Param0WriteEnabled=true
    Module.Param0WriteEnabled3=true
    Module.Param1WriteEnabled=true
    Module.Param1WriteEnabled3=true
    Module.Param2WriteEnabled=true
    Module.Param2WriteEnabled3=true
    Module.Param3WriteEnabled=true
    Module.Param3WriteEnabled3=true
    Write Parameter Index 3=true
   )
   &ScaleColor(
    Scale RGB=&VectorFromFloat(Value=&Add_Float(A=1, B=$System.FlareAll))
    ScaleA=false
    ScaleRGB=true
   ) {module=ScaleColor001}
   &ScaleColor(
    Scale Alpha=$System.FXBrightnessRamped
    Scale RGB=&VectorFromFloat(Value=$System.FXBrightnessRamped)
    ScaleA=true
    ScaleRGB=true
   ) {module=ScaleColor003}
  ]
  rendererClasses [&NiagaraSpriteRendererProperties]
  renderers [{rendererClass=&NiagaraSpriteRendererProperties}]
 }
 boneflow_radial_streaks_body {
  emitterUpdateScript [
   &EmitterState(
    Life Cycle Mode=::NewEnumerator1
    Loop Behavior=::NewEnumerator1
    Loop Duration Mode=::NewEnumerator1
    Scalability Mode=::NewEnumerator1
    UseLoopDelay=false
   )
   &SpawnRate(
    SpawnRate=&MakeCustomFloatFromBool(
     Bool=&SetBoolByFloatComparison(A=&VectorLength(Vector=$Engine.Owner.Velocity), B=50)
     True Float=0
     False Float=50
    )
    Use Spawn Probability=false
   ) {enabled=false}
   &SpawnBurst_Instantaneous(Spawn Count=15, Use Spawn Probability=false, UseLoopCountLimit=false)
  ]
  particleSpawnScript [
   &InitializeParticle(
    Color=<rgba 1 0.5 0 0.3>
    Lifetime=&RandomRangeFloat(Minimum=1, Maximum=1)
    Sprite Size Mode=::NewEnumerator3
    Uniform Sprite Size=25
    UsePositionOffset=false
    Write Material Random=true
   )
   /SetVariables(Particles.T:NiagaraFloat=&RemapRange(
     Input Value To Remap=&ReturnNormalizedExecIndex(Normalized Index Includes One=false)
     Input Min=0
     Input Max=1
     Output Min=0
     Output Max=1
    )) {isSetParameters=true, module=SetVariables_#10}
  ]
  particleUpdateScript [
   &ParticleState(Kill Particles When Lifetime Has Elapsed=false)
   &self:Pulse_Rays_01(
    Frequency=12
    Time:NiagaraFloat=$Engine.Time
    Time Offset:NiagaraFloat=&ReturnNormalizedExecIndex(Normalized Index Scale=12)
   ) {module=Pulse_Rays_01}
   &ScaleSpriteSize(
    Non-Uniform Curve Index:NiagaraFloat=&ReturnNormalizedExecIndex()
    Non-Uniform Curve Scale=<0.75 -6>
    Non-Uniform Curve Sprite Scale=DataInterface() {propertyValues {
      XCurve {keys [
        {arriveTangent=0.3555, interpMode=RCIM_Cubic, leaveTangent=0.3555, tangentMode=RCTM_User, time=0.2178, value=0.0429}
        {arriveTangent=0.3309, interpMode=RCIM_Cubic, leaveTangent=0.3309, tangentMode=RCTM_User, time=0.3583, value=0.4179}
        {arriveTangent=0.3661, interpMode=RCIM_Cubic, leaveTangent=0.3661, time=0.4649, value=0.4949}
        {arriveTangent=-1.2495, interpMode=RCIM_Cubic, leaveTangent=-1.2495, time=0.6856, value=0.5378}
        {arriveTangent=-0.0225, interpMode=RCIM_Cubic, leaveTangent=-0.0225, tangentMode=RCTM_User, time=0.9009, value=-0.05}
       ]}
      YCurve {keys [
        {arriveTangent=0.3555, interpMode=RCIM_Cubic, leaveTangent=0.3555, tangentMode=RCTM_User, time=0.2178, value=0.0429}
        {arriveTangent=0.3309, interpMode=RCIM_Cubic, leaveTangent=0.3309, tangentMode=RCTM_User, time=0.3583, value=0.4179}
        {arriveTangent=0.3661, interpMode=RCIM_Cubic, leaveTangent=0.3661, time=0.4649, value=0.4949}
        {arriveTangent=-1.1348, interpMode=RCIM_Cubic, leaveTangent=-1.1348, time=0.6856, value=0.5378}
        {arriveTangent=0.055, interpMode=RCIM_Cubic, leaveTangent=0.055, tangentMode=RCTM_User, time=0.9009, value=0}
       ]}
     }}
    Scale Sprite Size Mode=::NewEnumerator3
   ) {module=ScaleSpriteSize003}
   &FN_BoneFlow(
    Bias=0.3515
    Bone Offsets=DataInterface() {propertyValues {
      GpuSyncMode=SyncCpuToGpu
      MaxElements=0
      FloatData [<3 0 0>, <2 0 0>, <-1 0 0>, <3 0 0>, <0 0 0>, <0 0 0>, <0 0 0>, <0 0 0>, <0 0 0>, <0 0 0>, <0 0 0>, <0 0 0>]
     }}
    Bone Sets=DataInterface() {propertyValues {GpuSyncMode=SyncCpuToGpu, MaxElements=0, IntData [0,1,1,2,2,3,3,4]}}
    Enable Position Offsets=true
    Height:NiagaraFloat=$Particles.T
    SkeletalMesh:NiagaraDataInterfaceSkeletalMesh=$System.Skel
    Step Size:NiagaraFloat=&Divide_Float(A=0.1, B=1)
    Tension=0.3
    scale=1
   ) {module=FN_BoneFlow001}
   &SampleSkeletalMesh(
    AdditionalYawPitchRoll=false
    Bone / Socket Index=2
    Bone Sampling Mode:ENiagaraBoneSamplingMode=::NewEnumerator4
    "Interpolate Mesh Sampling (Update)"=true
    Position Offset Coordinate Space:ENiagaraCoordinateSpace=::World
    Skeletal Mesh:NiagaraDataInterfaceSkeletalMesh=$System.Skel
   )
   &self:Orient_01(Offset:Vector3f=<0 0 35>, RadialCenter:NiagaraPosition=$StackContext.SampleSkeletalMesh.SampledPosition, offsetScale=0.2) {enabled=false, module=Orient_01}
   &self:ScratchModule_01(
    Offset:Vector3f=<3 0 -25>
    RadialCenter:Vector3f=$Particles.SampleSkeletalMesh.SampledPosition
    offsetScale:NiagaraFloat=&FloatFromCurve(FloatCurve=DataInterface() {propertyValues {Curve {keys [
         {arriveTangent=0.5893, interpMode=RCIM_Cubic, leaveTangent=0.5893, tangentMode=RCTM_User, time=0, value=-0.2}
         {interpMode=RCIM_Cubic, tangentMode=RCTM_User, time=0.5, value=-0.05}
        ]}}})
   ) {module=ScratchModule_01}
   &CameraOffset(Camera Offset Amount=8)
   &ScaleSpriteSize(Scale Factor=&Vector2DFromFloat(Value=&FloatFromCurve(
      FloatCurve=DataInterface() {propertyValues {Curve {keys [<at 0 1 RCIM_Cubic>, {arriveTangent=-1.6315, interpMode=RCIM_Cubic, leaveTangent=-1.6315, tangentMode=RCTM_User, time=1, value=0}]}}}
      Scale Curve=$Particles.RandomCore
     ))) {enabled=false}
   &ScaleColor(
    Scale Alpha=&Multiply_Float(
     A=&FloatFromCurve(
      FloatCurve=DataInterface() {propertyValues {Curve {keys [<at 0.8 1>,<at 1 0>]}}}
      CurveIndex=&ReturnNormalizedExecIndex()
     )
     B=&FNTimeOfDayScalar(Night Scale=0.3, Morning Scale=0.1, Evening Scale=0.2)
    )
    Scale RGB=&VectorFromFloat(Value=&FloatFromCurve(
      FloatCurve=DataInterface() {propertyValues {Curve {keys [<at 0 0>,<at 1 1>]}}}
      CurveIndex=&ReturnNormalizedExecIndex(Normalized Index Scale=1.3)
      Scale Curve=&Lerp_Float(A=1, B=3, Alpha=$Particles.Pulse)
     ))
    ScaleA=true
    ScaleRGB=true
   )
   &ScaleColor(
    Scale Alpha=&FloatFromCurve(FloatCurve=DataInterface() {propertyValues {Curve {keys [
         {arriveTangent=9.0725, interpMode=RCIM_Cubic, leaveTangent=9.0725, tangentMode=RCTM_User, time=0, value=1}
         {arriveTangent=-16.4406, interpMode=RCIM_Cubic, leaveTangent=-16.4406, tangentMode=RCTM_User, time=0.1408, value=5}
         {arriveTangent=-0.0405, interpMode=RCIM_Cubic, leaveTangent=-0.0405, tangentMode=RCTM_User, time=0.7, value=1.3}
        ]}}})
    ScaleA=true
    ScaleRGB=false
   ) {module=ScaleColor002}
   &self:OffsetZ_01(Offset:Vector3f=<0 0 33>, Pivot:Vector3f=$Particles.SampleSkeletalMesh.SampledPosition, Scale=2) {module=OffsetZ_01}
   &DynamicMaterialParameters(
    Index 3 Param 1=$System.FlareAll
    Module.Param0WriteEnabled3=true
    Module.Param1WriteEnabled3=true
    Module.Param2WriteEnabled3=true
    Module.Param3WriteEnabled3=true
    Write Parameter Index 0=false
    Write Parameter Index 3=true
   )
   &ScaleColor(
    Scale RGB=&VectorFromFloat(Value=&Add_Float(A=1, B=$System.FlareAll))
    ScaleA=false
    ScaleRGB=true
   ) {module=ScaleColor001}
   &ScaleColor(
    Scale Alpha=$System.FXBrightnessRamped
    Scale RGB=&VectorFromFloat(Value=$System.FXBrightnessRamped)
    ScaleA=true
    ScaleRGB=true
   ) {module=ScaleColor003}
  ]
  rendererClasses [&NiagaraSpriteRendererProperties]
  renderers [
   {rendererClass=&NiagaraSpriteRendererProperties}
   {rendererClass=&NiagaraSpriteRendererProperties, rendererIndex=1}
  ]
 }
 single_halo001 {
  bEnabled=false
  emitterUpdateScript [
   &EmitterState(Life Cycle Mode=::NewEnumerator1, Loop Behavior=::NewEnumerator1, Loop Duration=25, UseLoopDelay=false)
   &SpawnBurst_Instantaneous(Spawn Time=0.65, Use Spawn Probability=false, UseLoopCountLimit=false)
  ]
  particleSpawnScript [&InitializeParticle(
    Color=<rgba 5 2 0 1>
    Lifetime=0.3
    Sprite Size Mode=::NewEnumerator3
    Uniform Sprite Size=75
    UsePositionOffset=false
    Write Material Random=true
   )]
  particleUpdateScript [
   &ParticleState(Kill Particles When Lifetime Has Elapsed=false)
   &SkeletalMeshLocation(
    Bone / Socket Index=7
    Bone Sampling Mode=::NewEnumerator4
    MaskSpawnGroup=false
    Orientation Sampling=::NewEnumerator1
    Sampled Position Offset=<8 -10 0>
    Skeletal Mesh=$System.Skel
   ) {module=SkeletalMeshLocation001}
   &CameraOffset(Camera Offset Amount=5)
   /SetVariables(Particles.SpriteSize=&Multiply_Vector2D_ByFloat(
     Float=&FloatFromCurve(
      FloatCurve=DataInterface() {propertyValues {Curve {keys [
          {arriveTangent=16.2037, interpMode=RCIM_Cubic, leaveTangent=16.2037, tangentMode=RCTM_User, time=0, value=0}
          {arriveTangent=4.6475, interpMode=RCIM_Cubic, leaveTangent=4.6475, tangentMode=RCTM_User, time=0.3, value=3.125}
          {arriveTangent=-0.0405, interpMode=RCIM_Cubic, leaveTangent=-0.0405, tangentMode=RCTM_User, time=1, value=4.375}
         ]}}}
      Scale Curve=43
     )
     Vector2D=<0.65 0.65>
    )) {enabled=false, isSetParameters=true, module=SetVariables_#9}
   /SetVariables(Particles.SpriteSize=&Multiply_Vector2D_ByFloat(
     Float=&FloatFromCurve(
      FloatCurve=DataInterface() {propertyValues {Curve {keys [
          {arriveTangent=34.6115, interpMode=RCIM_Cubic, leaveTangent=34.6115, tangentMode=RCTM_User, time=0, value=0}
          {arriveTangent=4.6475, interpMode=RCIM_Cubic, leaveTangent=4.6475, tangentMode=RCTM_User, time=0.2, value=3.75}
          <at 0.7 4.625>
          {arriveTangent=-0.0405, interpMode=RCIM_Cubic, leaveTangent=-0.0405, tangentMode=RCTM_User, time=1, value=4.375}
         ]}}}
      Scale Curve=43
     )
     Vector2D=<0.65 0.65>
    )) {isSetParameters=true}
   &ScaleColor(
    Scale Alpha=&FloatFromCurve(
     FloatCurve=DataInterface() {propertyValues {Curve {keys [
         {arriveTangent=-0.2933, interpMode=RCIM_Cubic, leaveTangent=-0.2933, tangentMode=RCTM_User, time=0, value=0}
         {arriveTangent=-2.8313, interpMode=RCIM_Cubic, leaveTangent=-2.8313, tangentMode=RCTM_User, time=0.1, value=0.95}
         {arriveTangent=-1.6622, interpMode=RCIM_Cubic, leaveTangent=-1.6622, tangentMode=RCTM_User, time=0.2, value=0.45}
         {arriveTangent=-0.0383, interpMode=RCIM_Cubic, leaveTangent=-0.0383, tangentMode=RCTM_User, time=1, value=0.2}
        ]}}}
     Scale Curve=0.4
    )
    ScaleA=true
    ScaleRGB=false
   ) {module=ScaleColor001}
  ]
  rendererClasses [&NiagaraSpriteRendererProperties, &NiagaraLightRendererProperties]
  renderers [
   {rendererClass=&NiagaraSpriteRendererProperties}
   {rendererClass=&NiagaraSpriteRendererProperties, rendererIndex=1}
   {rendererClass=&NiagaraLightRendererProperties, rendererIndex=2}
  ]
 }
 sparkle_center {
  bEnabled=false
  emitterUpdateScript [
   &EmitterState(Life Cycle Mode=::NewEnumerator1, Loop Behavior=::NewEnumerator1, Loop Duration Mode=::NewEnumerator1, UseLoopDelay=false)
   &SpawnBurst_Instantaneous(Spawn Count=44, Spawn Time=0.55, Use Spawn Probability=false, UseLoopCountLimit=false)
  ]
  particleSpawnScript [
   &InitializeParticle(
    Color Maximum=<rgba 1 0.5692 0.3873 1>
    Color Minimum=<rgba 1 0.6949 0.0328 1>
    Color Mode=::NewEnumerator2
    Lifetime=&RandomRangeFloat(Minimum=0.3, Maximum=1.2)
    Sprite Rotation Mode=::NewEnumerator2
    Sprite Size Mode=::NewEnumerator1
    Uniform Sprite Size Max=6
    Uniform Sprite Size Min=3.5
    UsePositionOffset=false
    Write Material Random=true
   )
   /SetVariables(Particles.RandomCore:NiagaraFloat=&FloatFromCurve(
     FloatCurve=DataInterface() {propertyValues {Curve {keys [
         {arriveTangent=0.0559, interpMode=RCIM_Cubic, leaveTangent=0.0559, tangentMode=RCTM_User, time=0, value=0.2}
         {arriveTangent=0.1895, interpMode=RCIM_Cubic, leaveTangent=0.1895, tangentMode=RCTM_User, time=0.9, value=0.25}
         {arriveTangent=1.1808, interpMode=RCIM_Cubic, leaveTangent=1.1808, tangentMode=RCTM_User, time=1, value=0.3}
        ]}}}
     CurveIndex=&RandomRangeFloat(Minimum=0, Maximum=1)
     Scale Curve=3
    )) {isSetParameters=true, module=SetVariables_#8}
   &SkeletalMeshLocation(
    Bone / Socket Index=7
    Bone Sampling Mode=::NewEnumerator4
    MaskSpawnGroup=false
    Orientation Sampling=::NewEnumerator1
    Sampled Position Offset=<10 0 0>
    Sampled Position Offset Coordinate Space:ENiagaraSkelMeshTransforms=::NewEnumerator1
    Skeletal Mesh=$System.Skel
   )
   &SphereLocation(
    MaskSpawnGroup=false
    Offset=<5 0 15>
    Override Seed=false
    Sphere Origin:NiagaraPosition=$Particles.SkeletalMeshLocation.SampledPosition
    Sphere Radius=15
    Surface Only=false
    UseNonUniformScale=false
    UseOffset=true
    UseSphereOrientation=false
   )
   &AddVelocityFromPoint(Module.Use Falloff Distance=false, Origin Offset:NiagaraPosition=$Output.SphereLocation.Position, Velocity Strength=222)
  ]
  particleUpdateScript [
   &ParticleState(Kill Particles When Lifetime Has Elapsed=false)
   &CurlNoiseForce(Noise Frequency=125, Noise Strength=&Multiply_Float(A=125, B=1), Pan Noise=true, Pan Noise Field=<0.5 0.5 0>)
   &ScaleSpriteSize(Scale Factor=&Vector2DFromFloat(Value=&FloatFromCurve(
      FloatCurve=DataInterface() {propertyValues {Curve {keys [
          <at 0 1 RCIM_Cubic>
          {arriveTangent=-1, interpMode=RCIM_Cubic, leaveTangent=-1, time=0.6, value=1.25}
          {arriveTangent=-0.1423, interpMode=RCIM_Cubic, leaveTangent=-0.1423, tangentMode=RCTM_User, time=1, value=0}
         ]}}}
      Scale Curve=$Particles.RandomCore
     )))
   &ScaleSpriteSize(
    Scale Factor=&Multiply_Vector2D(A=$Particles.SpriteSize, B=<0.3 0.3>)
    Scale Sprite Size Mode=::NewEnumerator2
   ) {module=ScaleSpriteSize001}
   &ScaleColor(
    Scale Alpha=&Multiply_Float(
     A=&FloatFromCurve(FloatCurve=DataInterface() {propertyValues {Curve {keys [
          {tangentMode=RCTM_User, time=0, value=1}
          {arriveTangent=-0.0405, interpMode=RCIM_Cubic, leaveTangent=-0.0405, tangentMode=RCTM_User, time=0.7009, value=1}
          {arriveTangent=-8.0089, interpMode=RCIM_Cubic, leaveTangent=-8.0089, tangentMode=RCTM_User, time=1, value=0}
         ]}}})
     B=&FNTimeOfDayScalar(Night Scale=0.3, Morning Scale=0.1, Evening Scale=0.2)
    )
    Scale RGB=&VectorFromFloat(Value=&FloatFromCurve(FloatCurve=DataInterface() {propertyValues {Curve {keys [
          {arriveTangent=2.9857, interpMode=RCIM_Cubic, leaveTangent=2.9857, tangentMode=RCTM_User, time=0, value=0}
          <at 0.1 2 RCIM_Cubic>
          {arriveTangent=-30.1063, interpMode=RCIM_Cubic, leaveTangent=-30.1063, tangentMode=RCTM_User, time=0.2, value=0.5}
         ]}}}))
    ScaleA=true
    ScaleRGB=true
   )
   &self:ScratchModule_03(AimPosition:NiagaraPosition=$Particles.SkeletalMeshLocation.SampledPosition) {module=ScratchModule_03}
   &Drag(Drag=2, Use Linear Drag=true, Use Rotational Drag=false)
   &SolveForcesAndVelocity(Clamp Velocity=false, Limit Acceleration=false)
  ]
  rendererClasses [&NiagaraSpriteRendererProperties]
  renderers [{rendererClass=&NiagaraSpriteRendererProperties}]
 }
}
summary {
 systemName=NS_VelvetMoth_Emote_A
 emitters [
  {emitterName=Body_Rays_Mesh, rendererClasses [&NiagaraMeshRendererProperties]}
  {emitterName=boneflow_radial_streaks_body, rendererClasses [&NiagaraSpriteRendererProperties]}
  {emitterName=boneflow_radial_peaks_B, rendererClasses [&NiagaraSpriteRendererProperties]}
  {emitterName=Body_Sparkles, rendererClasses [&NiagaraSpriteRendererProperties]}
  {
   bEnabled=false
   emitterName=single_halo001
   rendererClasses [&NiagaraSpriteRendererProperties, &NiagaraLightRendererProperties]
  }
  {
   bEnabled=false
   emitterName=beam_elements
   simTarget=GPUComputeSim
   rendererClasses [&NiagaraSpriteRendererProperties, &NiagaraMeshRendererProperties]
  }
  {bEnabled=false, emitterName=sparkle_center, rendererClasses [&NiagaraSpriteRendererProperties]}
  {bEnabled=false, emitterName=Counter_burst, rendererClasses [&NiagaraSpriteRendererProperties]}
  {
   bEnabled=false
   emitterName=Counter_nosim_infDigits
   rendererClasses [&NiagaraMeshRendererProperties, &NiagaraSpriteRendererProperties]
  }
 ]
 userVariables [
  User.Flare {type=NiagaraFloat, defaultValue=0, description=""}
  User.FXBrightnessScale {type=NiagaraFloat, defaultValue=0.86, description=""}
  User.InputDigits {type=NiagaraInt32, defaultValue=129}
  User.RampIn {type=NiagaraFloat, defaultValue=1, description=""}
  User.RampOut {type=NiagaraFloat, defaultValue=1, description=""}
 ]
}
